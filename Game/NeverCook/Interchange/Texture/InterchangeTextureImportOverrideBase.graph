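# Editor utility base Blueprint for texture import overrides that corrects a texture's LOD group based on its name.
# GetTextureGroupBySuffix matches name suffixes against TextureGroups, then GlobalTextureGroups, whose defaults map mask and specular-style suffixes to WorldSpecular.
# ProcessTexture starts from the current group, applies any suffix match, then remaps the result through OverrideMap.
# It writes LODGroup only when the result differs, returning whether it changed and the new group.
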EditorUtilityBlueprint /Game/NeverCook/Interchange/Texture/InterchangeTextureImportOverrideBase
graphs [
 EventGraph {links [], nodes []}
 GetTextureGroupBySuffix {nodes [
   GetTextureGroupBySuffix!() -> (then, TextureName:String) {is=title K2Node_FunctionEntry}
   > n1 = Sequence!(execute) -> (then_0, then_1, then_2:Exec)
   > n2 = "For Each Loop (Map)"!(
    execute:Exec
    MapPin:Map of TextureGroup Enums to Texture Group Affixes Structures=Get TextureGroups() -> (TextureGroups:Map of TextureGroup Enums to Texture Group Affixes Structures=TEXTUREGROUP_World)
    BreakPin:Exec
   ) -> (then:Exec, KeyPin:TextureGroup Enum, ValuePin:Texture Group Affixes Structure, CompletedPin:Exec)
   > Set l_group(l_group:TextureGroup Enum=@n2) -> (Output_Get:TextureGroup Enum)
   > n3 = For Each Loop(Array=Break Texture Group Affixes!(TextureGroupAffixes:"Texture Group Affixes Structure (by ref)"=@n2.ValuePin) -> (Suffixes_12_#1:Array of Strings) {is=title K2Node_BreakStruct})
   > Branch(Condition=EndsWith(SourceString=Get TextureName() -> (TextureName:String), InSuffix=@n3))
   > Return Node!(execute, Success:bool=true, Texture Group:TextureGroup Enum=Get l_group() -> (l_group:TextureGroup Enum))
   n4 = "For Each Loop (Map)"(
    execute=@n1.then_1
    MapPin=Get GlobalTextureGroups() -> (GlobalTextureGroups:Map of TextureGroup Enums to Texture Group Affixes Structures=TEXTUREGROUP_World)
   )
   > Set l_group(l_group=@n4)
   > n5 = For Each Loop(Array=Break Texture Group Affixes(TextureGroupAffixes=@n4.ValuePin))
   > Branch(Condition=EndsWith(SourceString=Get TextureName(), InSuffix=@n5))
   > Return Node(Success=true, Texture Group=Get l_group())
   Return Node(execute=@n1.then_2, Success=false, Texture Group=TEXTUREGROUP_World)
  ]}
 ProcessTexture {nodes [
   n1 = ProcessTexture!() -> (then, SourceTexture:Texture Object Reference) {is=title K2Node_FunctionEntry}
   > Set l_group(l_group=Get LODGroup!(self:Texture Object Reference=@n1) -> (LODGroup:TextureGroup Enum))
   n2 = CreateAssetData(InAsset=@n1, bAllowBlueprintClass=false)
   > Set l_name(l_name:String="To String (Name)"(InName=Break Asset Data!(AssetData=@n2) -> (AssetName))) -> (Output_Get:String)
   > n3 = Sequence()
   > n4 = GetTextureGroupBySuffix!(execute, self, TextureName:String=Get l_name() -> (l_name:String)) -> (then, Success:bool, Texture Group:TextureGroup Enum) {is=title K2Node_CallFunction}
   > Branch(Condition=@n4)
   > Set l_group(l_group=@n4.Texture Group)
   n5 = Find!(
    execute:Exec=@n3.then_1
    TargetMap:Map of TextureGroup Enums to TextureGroup Enums=Get OverrideMap() -> (OverrideMap:Map of TextureGroup Enums to TextureGroup Enums=TEXTUREGROUP_World)
    Key:"TextureGroup Enum (by ref)"=Get l_group()
   ) -> (then:Exec, Value:TextureGroup Enum=TEXTUREGROUP_World, ReturnValue)
   > Branch(Condition=@n5.ReturnValue)
   > Set l_group(l_group=@n5)
   n6 = Get SourceTexture() -> (SourceTexture:Texture Object Reference)
   n7 = Branch(
    execute=@n3.then_2
    Condition="Not Equal (Enum)"!(A:TextureGroup Enum=Get LODGroup(self=@n6), B:TextureGroup Enum=Get l_group()) -> (ReturnValue:bool)
   )
   > n8 = Set LODGroup!(execute, LODGroup:TextureGroup Enum=Get l_group(), self:Texture Object Reference=Get SourceTexture()) -> (then, Output_Get:TextureGroup Enum)
   > Return Node!(execute, WasOverridden:bool=true, TextureGroup:TextureGroup Enum=@n8)
   Return Node(execute=@n7.else, WasOverridden=false, TextureGroup=TEXTUREGROUP_World)
  ]}
]
variables [
 GlobalTextureGroups {default {TEXTUREGROUP_WorldSpecular {suffixes [_m,_s,_fx,_mask,_tangent,_scc,_ssc,_stt,_spec,_specular,_cm]}}}
 OverrideEnabled {type=boolean, default=true}
 OverrideMap {}
 TextureGroups {}
]
